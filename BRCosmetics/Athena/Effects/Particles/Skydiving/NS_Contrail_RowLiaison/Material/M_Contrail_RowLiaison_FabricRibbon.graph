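# Two-sided lit material for the RowLiaison skydiving contrail's fabric ribbons, built for Niagara ribbon and mesh particles.
# World position offset math combines optional axis twist with sine flapping, driven by dynamic parameters, particle randomness and an optional near-camera fade.
# Fresnel blends two pink tints, alongside a tiled fabric normal map and emissive adjustments scaled by particle color.
# A glitter term multiplies UV-space and screen-space dot textures, modulated by atmospheric sun direction.
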Material /BRCosmetics/Athena/Effects/Particles/Skydiving/NS_Contrail_RowLiaison/Material/M_Contrail_RowLiaison_FabricRibbon
graph {
 properties {
  bAutomaticallySetUsageInEditor=false
  bUsedWithNiagaraMeshParticles=true
  bUsedWithNiagaraRibbons=true
  twoSided=true
  usedShadingModels=Default Lit
 }
 nodes [
  n1 = Normalize(VectorInput=Transform(None=Constant3Vector() {constant=<rgba 0 0 1 0>}) {transformSourceType=TRANSFORMSOURCE_Instance})
  n2 = Add(
   A=Multiply(
    A=DynamicParameter() {parameterIndex=1, defaultValue=<rgba 1 1 1 0>, paramNames [TwistCurve,WPOCurve,Param3,Param4]}.TwistCurve
    B=ScalarParameter() {parameterName=TwistSpeed}
   )
   B=LinearInterpolate(Alpha=ParticleRandom())
  )
  n3 = MaterialFunctionCall() {materialFunction=&LinearGradient}
  n4 = LinearInterpolate(B=Multiply(A=Sine(None=@n2), B=ScalarParameter() {defaultValue=0.1, parameterName=TwistAmount}), Alpha=@n3.VGradient)
  n5 = MaterialFunctionCall() {materialFunction=&ObjectPivotPoint}
  n6 = WorldPosition()
  n7 = VectorParameter() {parameterName=WPO_Direction, defaultValue=<rgba 0.4 0 0 1>}
  n8 = OneMinus(None=MaterialFunctionCall() {materialFunction=&LinearGradient}.VGradient)
  n9 = Multiply(
   A=ComponentMask(None=VectorParameter() {parameterName=UV_Scale, defaultValue=<rgba 1 1 0 1>}.RGB) {g=true, r=true}
   B=LinearInterpolate(B=@n8, Alpha=ScalarParameter() {defaultValue=1, parameterName=Use_UV_VertexColor}) {constA=1}
  )
  n10 = ParticleRandom()
  n11 = DynamicParameter() {defaultValue=<rgba 0.5 1 1 1>, paramNames [FlappingSpeed, Time Offset, Flapping Time Offset Along Length, ""]}
  n12 = Multiply(
   A=Add(
    A=DynamicParameter() {parameterIndex=1, defaultValue=<rgba 1 1 1 0>, paramNames [TwistCurve,WPOCurve,Param3,Param4]}.WPOCurve
    B=LinearInterpolate(Alpha=@n10)
   )
   B=Multiply(A=@n11.Flapping Time Offset Along Length)
  ) {constB=2.8}
  n13 = DynamicParameter() {defaultValue=<rgba 0.5 1 1 1>, paramNames [FlappingSpeed, Time Offset, Flapping Time Offset Along Length, ""]}
  n14 = Add(A=Add(A=StaticComponentMaskParameter(None=@n9) {defaultG=true}, B=OneMinus(None=@n12)), B=@n13.FlappingSpeed)
  n15 = MaterialFunctionCall() {materialFunction=&LinearGradient}
  n16 = Multiply(A=Multiply(A=Sine(None=@n14), B=OneMinus(None=@n15.VGradient))) {constB=20}
  n17 = LinearInterpolate(Alpha=Add(A=@n8) {constB=0.5})
  n18 = Transform(None=Multiply(A=Multiply(A=@n7.RGB, B=@n16), B=Saturate(None=@n17))) {transformSourceType=TRANSFORMSOURCE_View}
  n19 = DynamicParameter() {defaultValue=<rgba 0.5 1 1 1>, paramNames [FlappingSpeed, Time Offset, Flapping Time Offset Along Length, ""]}
  n20 = Multiply(A=Multiply(A=Multiply(A=@n18) {constB=0.5}, B=@n19.Time Offset))
  n21 = ScalarParameter() {defaultValue=0.3, group=Near_Camera_fade, parameterName=Min_Near_Camera_WPO}
  n22 = MaterialFunctionCall(Distance=ScalarParameter() {defaultValue=256, group=Near_Camera_fade, parameterName=Near_Camera_fade}) {materialFunction=&Near_Camera_Fade}
  n23 = LinearInterpolate(
   B=Saturate(None=LinearInterpolate(A=@n21, Alpha=@n22.Result) {constA=0.5})
   Alpha=ScalarParameter() {group=Near_Camera_fade, parameterName=Use_Near_Camera_Fade}
  ) {constA=1}
  Add(
   A=LinearInterpolate(
    B=RotateAboutAxis(NormalizedRotationAxis=@n1, RotationAngle=@n4, PivotPoint=@n5.Object Pivot Location, Position=@n6.XYZ)
    Alpha=ScalarParameter() {parameterName=UseTwist}
   )
   B=LinearInterpolate(B=Multiply(A=@n20, B=@n23), Alpha=ScalarParameter() {defaultValue=1, parameterName=Use_WPO})
  )
  n24 = Fresnel() {exponent=0.5}
  n25 = OneMinus(None=@n24)
  LinearInterpolate(
   A=VectorParameter() {parameterName=Color_In, defaultValue=<rgba 0.9734 0.0612 0.3185 1>}.RGB
   B=VectorParameter() {parameterName=Color_Out, defaultValue=<rgba 1 0.4167 0.5761 1>}.RGB
   Alpha=Add(A=Power(Base=@n24) {constExponent=8}, B=Power(Base=@n25) {constExponent=10})
  )
  MaterialFunctionCall(Emissive=Multiply(A=ParticleColor().RGB, B=NamedRerouteUsage())) {materialFunction=&MF_FX_EmissiveAdjustments}
  n26 = DynamicParameter() {defaultValue=<rgba 0.5 1 1 1>, paramNames [FlappingSpeed, Time Offset, Flapping Time Offset Along Length, ""]}
  n27 = Add(A=Time(), B=@n26.Time Offset)
  n28 = ComponentMask(None=TextureCoordinate() {coordinateIndex=2}) {g=true}
  n29 = Sine(None=Add(A=Multiply(A=@n27, B=@n26.FlappingSpeed), B=Multiply(A=@n28, B=@n26.Flapping Time Offset Along Length)))
  n30 = Multiply(A=ScalarParameter() {defaultValue=0.007, parameterName=Rotation Range}, B=@n26.Flapping Time Offset Along Length)
  NamedRerouteDeclaration(None=RotateAboutAxis(
    NormalizedRotationAxis=NamedRerouteUsage()
    RotationAngle=Multiply(A=@n29, B=@n30) {constB=0}
    PivotPoint=MaterialFunctionCall() {materialFunction=&ComponentPivotLocation}.Result
    Position=WorldPosition().XYZ
   )) {name=WPO, nodeColor=<rgba 0 0.1094 1 1>}
  n31 = Multiply(A=TextureCoordinate()) {constB=3}
  n32 = ScreenPosition()
  n33 = Divide(A=MaterialFunctionCall() {materialFunction=&ScreenResolution}.Visible Resolution) {constB=256}
  n34 = Multiply(
   A=Power(Base=TextureSampleParameter2D(UVs=@n31) {parameterName=Texture_Main, texture=/Game/Effects/Fort_Effects/Textures/Tile/T_Dots_Small}.R)
   B=TextureSampleParameter2D(UVs=Multiply(A=@n32.ViewportUV, B=@n33)) {parameterName=Texture_Secondary, texture=&T_Dots_Small}.R
  )
  n35 = Multiply(A=@n34) {constB=500}
  n36 = Multiply(A=@n34) {constB=1000}
  n37 = TwoSidedSign()
  n38 = ScalarParameter() {defaultValue=0.05, parameterName=MinLight}
  n39 = ScalarParameter() {defaultValue=1, parameterName=MaxLight}
  n40 = AtmosphericLightVector()
  n41 = VertexNormalWS()
  n42 = Power(Base=Saturate(None=DotProduct(A=@n40, B=@n41))) {constExponent=5}
  NamedRerouteDeclaration(None=Multiply(A=LinearInterpolate(A=@n38, B=@n39, Alpha=@n42), B=LinearInterpolate(A=@n35, B=@n36, Alpha=@n37) {constB=0})) {name=Glitter, nodeColor=<rgba 0.8984 0 1 1>}
  n43 = ComponentMask(None=Subtract(A=CameraPositionWS(), B=ParticlePositionWS())) {g=true, r=true}
  NamedRerouteDeclaration(None=AppendVector(A=Normalize(VectorInput=@n43), B=Constant())) {name=Rotation Axis, nodeColor=<rgba 0.6875 1 0 1>}
  NamedRerouteUsage()
  OneMinus(None=@n3.VGradient)
  ScalarParameter() {defaultValue=2, parameterName=Flapping Speed}
  ScalarParameter() {defaultValue=2, parameterName=Flapping Time Offset Along Length}
  TextureSampleParameter2D(UVs=Multiply(B=TextureCoordinate()) {constA=1, constB=3}) {
   parameterName=Param
   samplerType=SAMPLERTYPE_Normal
   texture=/BRCosmetics/Characters/Player/Male/Medium/Bodies/M_MED_PureCereal/Textures/T_PureCereal_tileableFabric_PureCerealFabric_N
  }
 ]
}
