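# Niagara skydiving contrail cosmetic with paired cloth and airflow trails, glitter and camera-flash bursts.
# AirFlow, cloth and Glitter follow a system socket reader and alignment direction, using per-trail index to splay each side.
# Sparkles ride the cloth by sampling its static mesh, rotated by the cloth quaternion, with spawning tied to dive alpha.
# PhotoFlashes layers sprites, cone meshes and lights; front-end and time-of-day modifiers tune the look.
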NiagaraSystem /BRCosmetics/Athena/Effects/Particles/Skydiving/NS_Contrail_RowLiaison/NS_Contrail_RowLiaison_02
dependencies {
 usedDataInterfaces [
  &NiagaraDataInterfaceCurve
  &NiagaraDataInterfaceArrayFloat3
  &NiagaraDataInterfaceArrayFloat
  &NiagaraDataInterfaceStaticMesh
  &NiagaraDataInterfaceColorCurve
  &NiagaraDataInterfaceSocketReader
 ]
 usedDynamicInputs [
  &Vector2DFromFloat
  &MaskIntBySpawnGroup
  &ReturnExecIndex
  &Lerp_Vector
  &TransformVector
  &Add_Float
  &Sine
  &Multiply_VectorByFloat
  /Niagara/DynamicInputs/Math/Clamp_Vector
 ]
 usedModules [
  &self:SystemSpawnScript.SystemScriptSource.SystemScriptGraph.NiagaraNodeAssignment_0.SetVariables_#1
  &self:SystemSpawnScript.SystemScriptSource.SystemScriptGraph.NiagaraNodeAssignment_4.SetVariables_#2
  &SystemState
  &self:SystemSpawnScript.SystemScriptSource.SystemScriptGraph.NiagaraNodeAssignment_6.SetVariables_#3
  &self:SystemSpawnScript.SystemScriptSource.SystemScriptGraph.NiagaraNodeAssignment_7.SetVariables_#4
  &Timeline
  &self:SystemSpawnScript.SystemScriptSource.SystemScriptGraph.NiagaraNodeAssignment_3.SetVariables_#5
  &self:SystemSpawnScript.SystemScriptSource.SystemScriptGraph.NiagaraNodeAssignment_8.SetVariables_#6
 ]
}
emitters {
 AirFlow {
  emitterUpdateScript [
   &EmitterState(
    Inactive Response=::NewEnumerator1
    Life Cycle Mode=::NewEnumerator1
    Loop Behavior=::NewEnumerator1
    Loop Duration Mode=::NewEnumerator1
    UseLoopDelay=false
   )
   &SpawnBurst_Instantaneous(Spawn Count=2, Use Spawn Probability=false, UseLoopCountLimit=false)
  ]
  particleSpawnScript [
   &PartitionParticles()
   &InitializeParticle(
    Color Mode=::NewEnumerator0
    Lifetime=&RandomRangeFloat(Minimum=1.7, Maximum=2)
    Mesh Scale=&FN_FrontEndVector3Modifier(Gameplay=<0.75 0.75 1>, Front End=<0.75 0.75 1>)
    Mesh Scale Mode=::NewEnumerator4
    UsePositionOffset=false
    Write Material Random=true
   )
   &InitialMeshOrientation(
    Mesh Axis To Orient=&FN_FrontEndVector3Modifier(Gameplay=<0 0 1>, Front End=<0 0 1>)
    Mesh Orientation Mode=::NewEnumerator2
    Rotation=&FN_FrontEndVector3Modifier(
     Gameplay=&SelectVectorFromArray(
      Direct Array Index=$Particles.PartitionParticles.Partition
      Vector Selection Array=DataInterface() {propertyValues {FloatData [<0.03 0.01 0.09>, <0.03 -0.01 -0.09>, <0 0 0>, <0 0 0>]}}
     )
     Front End=&SelectVectorFromArray(
      Direct Array Index=$Particles.PartitionParticles.Partition
      Vector Selection Array=DataInterface() {propertyValues {FloatData [<0.05 0.03 0.09>, <0.05 -0.03 -0.09>, <0 0 0>, <0 0 0>]}}
     )
    )
    Rotation Coordinate Space=::NewEnumerator2
    Use Rotation Vector=false
   ) {module=InitialMeshOrientation001}
  ]
  particleUpdateScript [
   &ParticleState(Kill Particles When Lifetime Has Elapsed=false, Let Infinitely Lived Particles Die When Emitter Deactivates=true)
   &UpdateMeshOrientation(
    Facing Direction=&Add_Vector(
     A=$System.AlignmentDirection
     B=&SelectVectorFromArray(
      Direct Array Index=$Particles.PartitionParticles.Partition
      Direct Array Mode=::NewEnumerator1
      Vector Selection Array=DataInterface() {propertyValues {FloatData [<0.2 0 0>, <-0.2 0 0>]}}
     )
    )
    Facing Interpolation=false
    Orientation Method=::NewEnumerator1
    Override Quat=false
    Stabilization Mode:ENiagara_OrientVectorCount=::NewEnumerator2
    TransformFacing=false
    TransformSide=false
   ) {module=UpdateMeshOrientation002}
   &ScaleMeshSize(Scale Factor=&MakeVector(
     X=1
     Y=1
     Z=&FloatFromCurve(
      FloatCurve=DataInterface() {propertyValues {Curve {keys [
          {arriveTangent=0.9959, interpMode=RCIM_Cubic, leaveTangent=0.9959, tangentMode=RCTM_User, time=0, value=0}
          {interpMode=RCIM_Cubic, tangentMode=RCTM_User, time=1, value=1}
         ]}}}
      CurveIndex=$Emitter.Age
     )
    ))
   &DynamicMaterialParameters(
    Index 0 Param 2=$System.Dive
    Index 0 Param 3=&RandomRangeFloat(Minimum=0, Maximum=1)
    Index 0 Param 4=&RandomRangeFloat(Minimum=0, Maximum=1)
    Module.Index0Param0Write=true
    Module.Index0Param1Write=true
    Module.Index0Param2Write=true
    Module.Index0Param3Write=true
   ) {module=DynamicMaterialParameters001}
   &SocketLocation(
    Orientation Sampling=::NewEnumerator1
    SamplePosOffset=false
    Socket Index=$Particles.PartitionParticles.Partition
    Socket Reader=$System.Socket Reader
    Socket Sampling Mode=::NewEnumerator3
    VelocitySpeedLimit=false
   )
   &ScaleMeshSize(Scale Factor=&FN_FrontEndVector3Modifier(Gameplay=<0.7 0.7 0.8>, Front End=<0.65 0.65 0.75>)) {module=ScaleMeshSize001}
   &ScaleColor(
    Scale Alpha=&FloatFromCurve(
     FloatCurve=DataInterface() {propertyValues {Curve {keys [
         {interpMode=RCIM_Cubic, tangentMode=RCTM_User, time=0.4, value=0}
         {arriveTangent=0.9919, interpMode=RCIM_Cubic, leaveTangent=0.9919, tangentMode=RCTM_User, time=0.8, value=1}
        ]}}}
     CurveIndex=$System.Timeline.Time
     Scale Curve=0.25
    )
    Scale RGB=&MakeVectorFromLinearColorRGB(Color=&FNTimeOfDayColor(
      Night Color=<rgba 0.1611 0.3212 0.9 1>
      Morning Color=<rgba 1 0.7571 0.6766 1>
      Day Color=<rgba 0.9 0.9 0.9 1>
      Evening Color=<rgba 0.8 0.4131 0.32 1>
     ))
    ScaleA=true
    ScaleRGB=true
   )
   &ScaleColor(Scale Alpha=&FN_FrontEndFloatModifier(Gameplay=1, Front End=0.6), ScaleA=true, ScaleRGB=true) {module=ScaleColor001}
  ]
  rendererClasses [&NiagaraMeshRendererProperties]
  renderers [{rendererClass=&NiagaraMeshRendererProperties, rendererIndex=0}]
 }
 ClothandSparkles {
  emitterUpdateScript [
   &EmitterState(
    Inactive Response=::NewEnumerator1
    Life Cycle Mode=::NewEnumerator1
    Loop Behavior=::NewEnumerator1
    Loop Duration Mode=::NewEnumerator1
    UseLoopDelay=false
   )
   &SpawnBurst_Instantaneous(Spawn Count=2, Use Spawn Probability=false, UseLoopCountLimit=false)
   &SpawnRate(Spawn Probability=$System.ContrailDiveAlpha, SpawnGroup=1, SpawnRate=10, Use Spawn Probability=true)
   &SpawnBurst_Instantaneous(Spawn Count=2, Spawn Time=3.3333, Use Spawn Probability=false, UseLoopCountLimit=false) {module=SpawnBurst_Instantaneous001}
  ]
  particleSpawnScript [
   &PartitionParticles()
   &InitializeParticle(
    Lifetime Max=1.3
    Lifetime Min=0.3
    Lifetime Mode=::NewEnumerator1
    Mesh Scale Mode=::NewEnumerator3
    Position Mode=::NewEnumerator0
    Sprite Size Max=<33 66>
    Sprite Size Min=<17 33>
    Sprite Size Mode=::NewEnumerator2
    Write Material Random=true
   )
   /SetVariables(
    Particles.CameraOffset:NiagaraFloat=&MaskFloatBySpawnGroup(Masked Float=0, Passthrough Float=10)
    Particles.VisibilityTag=&SpawnGroup()
   ) {isSetParameters=true}
   &InitialMeshOrientation(Mesh Axis To Orient=<0 0 1>, Mesh Orientation Mode=::NewEnumerator2, Use Rotation Vector=false)
  ]
  particleUpdateScript [
   /SetVariables(Particles.Flicker:NiagaraFloat=&Waveform(
     Global Input Position=&RandomRangeFloat(Minimum=0, Maximum=9)
     "Waveform [1]"=::NewEnumerator10
     "[1] Frequency"=9
     "[1] Amplitude Min/Max"=<0.6 1>
    )) {isSetParameters=true}
   &ParticleState(Kill Particles When Lifetime Has Elapsed=false, Let Infinitely Lived Particles Die When Emitter Deactivates=true)
   &ScaleSpriteSize(
    Uniform Curve Scale=$Particles.Flicker
    Uniform Curve Sprite Scale=DataInterface() {propertyValues {Curve {keys [
        {arriveTangent=9.0725, interpMode=RCIM_Cubic, leaveTangent=9.0725, tangentMode=RCTM_User, time=0, value=0}
        {arriveTangent=-2.5249, interpMode=RCIM_Cubic, leaveTangent=-2.5249, tangentMode=RCTM_User, time=0.1408, value=0.9847}
        {arriveTangent=-0.0405, interpMode=RCIM_Cubic, leaveTangent=-0.0405, tangentMode=RCTM_User, time=1, value=0}
       ]}}}
   )
   &ScaleMeshSize(Scale Factor=&VectorFromFloat(Value=&FloatFromCurve(
      FloatCurve=DataInterface() {propertyValues {Curve {keys [
          {arriveTangent=1.7685, interpMode=RCIM_Cubic, leaveTangent=1.7685, tangentMode=RCTM_User, time=0.1, value=0}
          {arriveTangent=0.0224, interpMode=RCIM_Cubic, leaveTangent=0.0224, tangentMode=RCTM_User, time=1, value=1}
         ]}}}
      CurveIndex=$Emitter.Age
     )))
   &KillParticles(Kill Particles=/Niagara/DynamicInputs/Helpers/MaskBoolBySpawnGroup(
     Masked Bool:NiagaraBool=false
     Passthrough Bool:NiagaraBool=&SetBoolByFloatComparison(A=$Particles.Age, B=$Particles.Lifetime)
     Spawn Group Mask:NiagaraInt32=0
    ))
   &SocketLocation(
    Orientation Sampling=::NewEnumerator1
    Position Sampling=::NewEnumerator1
    SamplePosOffset=false
    Socket Index=$Particles.PartitionParticles.Partition
    Socket Reader=$System.Socket Reader
    Socket Sampling Mode=::NewEnumerator3
    VelocitySpeedLimit=false
   )
   &UpdateMeshOrientation(
    Facing Direction=&Add_Vector(
     A=$System.AlignmentDirection
     B=&SelectVectorFromArray(
      Direct Array Index=$Particles.PartitionParticles.Partition
      Vector Selection Array=DataInterface() {propertyValues {FloatData [<0.2 0 0>, <-0.2 0 0>]}}
     )
    )
    Facing Interpolation=false
    Orientation Method=::NewEnumerator1
    Override Quat=false
    Stabilization Mode:ENiagara_OrientVectorCount=::NewEnumerator2
    TransformFacing=false
    TransformSide=false
   )
   /SetVariables(Particles.ClothQuaternion:Quat4f=&MultiplyQuaternion(
     Quaternion A=$Particles.MeshOrientation
     Quaternion B=&MakeQuatFromAxisAngle(
      Axis=<0 0 1>
      AngleInDegrees=&MergeHelper_FloatSwitch(0=$System.ClothRotationAngleA, 1=$System.ClothRotationAngleB, 2=?, 3=?, Index=$Particles.PartitionParticles.Partition, Num=::NewEnumerator0)
     )
    )) {isSetParameters=true}
   &DynamicMaterialParameters(
    Index 0 Param 1=2
    Index 0 Param 2=&SelectFloatFromArray(Direct Array Index=$Particles.PartitionParticles.Partition, Float Selection Array=DataInterface() {propertyValues {FloatData [0.33,0.333]}})
    Index 0 Param 3=2
    Module.Index0Param0Write=true
    Module.Index0Param1Write=true
    Module.Index0Param2Write=true
    Module.Index0Param3Write=false
   ) {module=DynamicMaterialParameters001}
   /Niagara/Modules/Spawn/Location/StaticMeshLocation(
    MaskSpawnGroup=false
    Mesh Sampling Type:ENiagara_StaticSamplingMode=::NewEnumerator2
    OffsetAlongNormal=false
    Orientation Sampling:ENiagara_AttributeSamplingApplyOutput=::NewEnumerator1
    OverrideRandomSeed=false
    Position Sampling:ENiagaraSkelMeshPositionSamplingMode=::NewEnumerator2
    SamplePosOffset=false
    Sampled Position Offset:Vector3f=$Output.SocketLocation.SampledPosition
    Static Mesh=DataInterface() {propertyValues {
      DefaultMesh=/BRCosmetics/Athena/Effects/Particles/Skydiving/NS_Contrail_PiperShelf/Meshes/SM_Contrail_PiperShelf_Cloth
      InstanceIndex=-1
      LODIndex=0
      PreviewMesh=&SM_Contrail_PiperShelf_Cloth
      SoftSourceActor=null
      SourceMode=DefaultMeshOnly
      bAllowSamplingFromStreamingLODs=false
      bCaptureTransformsPerFrame=true
      bUsePhysicsBodyVelocity=false
      FilteredSockets []
      LODIndexUserParameter {parameter {name=None, typeDefHandle {registeredTypeIndex=108}, varData []}}
      MeshParameterBinding {parameter {name=None, typeDefHandle {registeredTypeIndex=1}, varData []}}
      SectionFilter {allowedMaterialSlots []}
     }}
   )
   /SetVariables(Particles.Position=&MaskVectorBySpawnGroup(
     Masked Vector=$Output.SocketLocation.SampledPosition
     Passthrough Vector=&AddVectorToPosition(
      Position=$Output.SocketLocation.SampledPosition
      Vector=&RotateVectorByQuaternion(Vector To Rotate=$Particles.StaticMeshLocation.SampledPosition, Quaternion=$Particles.ClothQuaternion)
     )
    )) {isSetParameters=true}
   &ScaleColor(
    Scale Alpha=&FloatFromCurve(
     FloatCurve=DataInterface() {propertyValues {Curve {keys [
         {arriveTangent=9.0725, interpMode=RCIM_Cubic, leaveTangent=9.0725, tangentMode=RCTM_User, time=0, value=0}
         {arriveTangent=-2.5249, interpMode=RCIM_Cubic, leaveTangent=-2.5249, tangentMode=RCTM_User, time=0.1408, value=0.9847}
         {arriveTangent=-0.0405, interpMode=RCIM_Cubic, leaveTangent=-0.0405, tangentMode=RCTM_User, time=1, value=0}
        ]}}}
     Scale Curve=&Multiply_Float(A=$Particles.Flicker, B=$System.Timeline.Time)
    )
    Scale RGB=&MakeVectorFromLinearColorRGB(Color=&FNTimeOfDayColor(Night Color=<rgba 0.17 0.6127 1 1>, Morning Color=<rgba 1 1 1 1>, Day Color=<rgba 1 1 1 1>, Evening Color=<rgba 1 1 1 1>))
    ScaleA=true
    ScaleRGB=true
   )
   /SetVariables(Particles.BackSparkleColor:LinearColor=&MultiplyLinearColor(Linear Color=$Particles.Color, Scale Factor=&MakeVector4_FromFloat(Float=0.1))) {isSetParameters=true}
   &ScaleColor(
    Scale RGB=&FN_FrontEndVector3Modifier(
     Gameplay=&MakeVectorFromLinearColorRGB(Color=<rgba 1 0.6927 0.853 1>)
     Front End=&MakeVectorFromLinearColorRGB(Color=<rgba 0.5625 0.0791 0.3313 1>)
    )
    ScaleA=true
    ScaleRGB=true
   ) {module=ScaleColor001}
  ]
  rendererClasses [&NiagaraMeshRendererProperties, &NiagaraSpriteRendererProperties]
  renderers [
   {rendererClass=&NiagaraMeshRendererProperties, rendererIndex=0}
   {rendererClass=&NiagaraSpriteRendererProperties, rendererIndex=1}
   {rendererClass=&NiagaraSpriteRendererProperties, rendererIndex=2}
  ]
 }
 Glitter {
  emitterUpdateScript [
   &EmitterState()
   &SpawnRate(SpawnRate=30, Use Spawn Probability=false)
  ]
  particleSpawnScript [
   &FN_SetSpawnGroup(Number of Spawn Groups:NiagaraInt32=&MakeIntFromFloat(
     Float=&ClampFloat(Float=&Lerp_Float(A=4, B=-2000, Alpha=$User.ParaGlideLeanAlpha), Min=2, Max=2)
     Conversion Mode=::NewEnumerator1
    ))
   &InitializeParticle(
    Lifetime Max=1
    Lifetime Min=0.8
    Lifetime Mode=::NewEnumerator1
    Sprite Rotation Mode=::NewEnumerator2
    Sprite Size Mode=::NewEnumerator3
    Sprite UV Mode=::NewEnumerator3
    Uniform Sprite Size=&Multiply_Float(
     A=&RandomRangeFloat(Minimum=20, Maximum=30)
     B=&self:SpawnGrp_0_1_Mask(Masked Float:NiagaraFloat=1, Passthrough Float:NiagaraFloat=1.5)
    )
    UsePositionOffset=false
    Write Material Random=true
   )
   &SocketLocation(
    SamplePosOffset=true
    Socket Index=$Particles.SpawnGroup
    Socket Reader=$System.Socket Reader
    Socket Sampling Mode=::NewEnumerator3
    VelocitySpeedLimit=false
   )
   &ShapeLocation(
    Cylinder Height=33.8933
    Cylinder Height Midpoint=-0.1
    Cylinder Radius=15
    MaskSpawnGroup=false
    Override Seed=false
    Rotation Axis=&Add_Vector(
     A=$System.AlignmentDirection
     B=&SelectVectorFromArray(
      Direct Array Index=$Particles.SpawnGroup
      Direct Array Mode=::NewEnumerator1
      Vector Selection Array=DataInterface() {propertyValues {FloatData [<0.2 0 0>, <-0.2 0 0>]}}
     )
    )
    Rotation Mode=::NewEnumerator1
    Shape Primitive=::NewEnumerator1
    Spawn Group Mask=1
    Surface Only=false
    UseOwnerScale=true
   )
   &CameraOffset(Camera Offset Amount=&RandomRangeFloat(Minimum=0, Maximum=30))
   &AddVelocity(
    Cone Angle=30
    Cone Axis=&Add_Vector(
     A=$System.AlignmentDirection
     B=&SelectVectorFromArray(
      Direct Array Index=$Particles.SpawnGroup
      Direct Array Mode=::NewEnumerator1
      Vector Selection Array=DataInterface() {propertyValues {FloatData [<0.2 0 0>, <-0.2 0 0>]}}
     )
    )
    Module.UseConeFalloff=false
    Module.UseVelDistribution=false
    Override Seed=false
    Velocity Mode=::NewEnumerator2
    Velocity Speed=&FN_FrontEndFloatModifier(
     Gameplay=&RandomRangeFloat(Minimum=100, Maximum=200)
     Front End=&RandomRangeFloat(Minimum=75, Maximum=150)
    )
   )
  ]
  particleUpdateScript [
   &ParticleState()
   &ScaleColor(
    Scale Alpha=&FloatFromCurve(
     FloatCurve=DataInterface() {propertyValues {Curve {
        defaultValue=1
        keys [
         {interpMode=RCIM_Cubic, tangentMode=RCTM_User, time=0, value=1}
         {arriveTangent=-3.5748, interpMode=RCIM_Cubic, leaveTangent=-3.5748, tangentMode=RCTM_User, time=1, value=0}
        ]
       }}}
     Scale Curve=3
    )
    Scale RGB=&MakeVectorFromLinearColorRGB(Color=&ColorFromCurve(ColorCurve=DataInterface() {propertyValues {
        AlphaCurve {keys [<at 0.5984 1 RCIM_Cubic>,<at 1 0 RCIM_Cubic>]}
        BlueCurve {keys [<at 0.1101 5.2706>,<at 0.1432 158.1172>,<at 0.2806 5.2706>,<at 0.5454 5.2706>,<at 0.5785 158.1172>,<at 0.7226 5.2706>]}
        GreenCurve {keys [<at 0.1101 0.1109>,<at 0.1432 3.3267>,<at 0.2806 0.1109>,<at 0.5454 0.1109>,<at 0.5785 3.3267>,<at 0.7226 0.1109>]}
        RedCurve {keys [<at 0.1101 10>,<at 0.1432 300>,<at 0.2806 10>,<at 0.5454 10>,<at 0.5785 300>,<at 0.7226 10>]}
       }}))
    ScaleA=true
    ScaleRGB=true
   ) {module=ScaleColor001}
   &ScaleColor(
    Scale Alpha=&FloatFromCurve(
     FloatCurve=DataInterface() {propertyValues {Curve {keys [
         {interpMode=RCIM_Cubic, leaveTangentWeight=0.0833, tangentMode=RCTM_User, time=0, value=0.2}
         {
          arriveTangent=0.7559
          arriveTangentWeight=0.1572
          interpMode=RCIM_Cubic
          leaveTangent=0.7559
          leaveTangentWeight=0.1572
          tangentMode=RCTM_User
          time=0.25
          value=0.25
         }
         {arriveTangentWeight=0.1681, interpMode=RCIM_Cubic, leaveTangentWeight=0.0841, tangentMode=RCTM_User, time=0.5, value=1}
         {
          arriveTangent=-0.6245
          arriveTangentWeight=0.1572
          interpMode=RCIM_Cubic
          leaveTangent=-0.6245
          leaveTangentWeight=0.1572
          tangentMode=RCTM_User
          time=0.75
          value=0.25
         }
         {arriveTangentWeight=0.0833, interpMode=RCIM_Cubic, tangentMode=RCTM_User, time=1, value=0.2}
        ]}}}
     CurveIndex=Linked() {linkedVariable {name=NPC.FortniteNPC.TimeOfDay_0-1, type=NiagaraFloat}}
     Scale Curve=&FNTimeOfDayScalar(Night Scale=0.8, Morning Scale=2, Evening Scale=1)
    )
    ScaleA=true
    ScaleRGB=false
   ) {module=ScaleColor002}
   &DynamicMaterialParameters(
    Index 0 Param 1=&FloatFromCurve(FloatCurve=DataInterface() {propertyValues {Curve {keys [<at 0 0 RCIM_Cubic>,<at 1.0011 0.2>]}}})
    Module.Index0Param0Write=true
    Module.Index0Param1Write=true
    Module.Index0Param2Write=true
    Module.Index0Param3Write=true
   )
   &SolveForcesAndVelocity(Clamp Velocity=false, Limit Acceleration=false)
  ]
  rendererClasses [&NiagaraSpriteRendererProperties]
  renderers [{rendererClass=&NiagaraSpriteRendererProperties}]
 }
 PhotoFlashes {
  emitterUpdateScript [
   &EmitterState(
    Inactive Response=::NewEnumerator1
    Life Cycle Mode=::NewEnumerator1
    Loop Behavior=::NewEnumerator1
    Loop Duration Mode=::NewEnumerator1
    UseLoopDelay=true
   )
   &SpawnRate(SpawnRate=5, Use Spawn Probability=false) {module=SpawnRate001}
   &SpawnRate(
    SpawnRate=&FloatFromCurve(
     FloatCurve=DataInterface() {propertyValues {Curve {keys [
         {arriveTangent=4.8497, interpMode=RCIM_Cubic, leaveTangent=4.8497, tangentMode=RCTM_User, time=0, value=0}
         <at 0.6 1 RCIM_Cubic>
         {arriveTangent=-5.3133, tangentMode=RCTM_User, time=0.8841, value=0}
        ]}}}
     CurveIndex=$Emitter.NormalizedLoopAge
     Scale Curve=5
    )
    Use Spawn Probability=false
   ) {enabled=false}
  ]
  particleSpawnScript [
   &InitializeParticle(
    Lifetime Max=0.17
    Lifetime Min=0.13
    Lifetime Mode=::NewEnumerator1
    Mass Max=2
    Mass Mode=::NewEnumerator3
    Position Offset=<0 0 100>
    Position Offset Coordinate Space=::Local
    Sprite Rotation Mode=::NewEnumerator2
    Sprite Size Mode=::NewEnumerator3
    Uniform Sprite Size=&RandomRangeFloat(Minimum=10, Maximum=25)
    UsePositionOffset=true
    Write Material Random=true
   )
   &CameraOffset(Camera Offset Amount=20) {module=CameraOffset001}
   &ShapeLocation(
    MaskSpawnGroup=false
    Override Seed=false
    Rotation Angle=1
    Rotation Axis=<0 0 1>
    Rotation Mode=::NewEnumerator1
    Sphere Radius=150
    Sphere Surface Distribution=0.75
    UseOwnerScale=true
   ) {module=ShapeLocation001}
   /SetVariables(Particles.LightVisability:NiagaraInt32=&RandomRangeInt(Minimum=0, Maximum=1)) {isSetParameters=true}
   /SetVariables(Particles.Gen5LightDirection:Vector3f=<-15 -90 0>) {isSetParameters=true}
   &InitialMeshOrientation(
    Mesh Axis To Orient=<0 0 -1>
    Mesh Orientation Mode=::NewEnumerator2
    Orientation Vector=&NormalizeVector(
     Vector To Normalize=&Subtract_Vector(
      A=&AddVectorToPosition(Position=/Niagara/DynamicInputs/Helpers/SimulationPosition(), Vector=<0 0 140>)
      B=Linked() {linkedVariable {name=Particles.Position, type=NiagaraPosition}}
     )
     Safe Normalize=true
     Safe Fallback Threshold=0
     Safe Fallback Vector=<0 0 0.001>
    )
    Use Rotation Vector=true
   )
  ]
  particleUpdateScript [
   &ParticleState()
   &ScaleColor(
    Scale Alpha=&FloatFromCurve(
     FloatCurve=DataInterface() {propertyValues {Curve {keys [
         <at 0 0 RCIM_Cubic>
         {arriveTangent=0.1895, interpMode=RCIM_Cubic, leaveTangent=0.1895, tangentMode=RCTM_User, time=0.1, value=1}
         {arriveTangent=-0.0202, interpMode=RCIM_Cubic, leaveTangent=-0.0202, tangentMode=RCTM_User, time=0.35, value=1}
         <at 1.0011 0.002 RCIM_Cubic>
        ]}}}
     Scale Curve=&FN_FrontEndFloatModifier(Gameplay=9, Front End=6)
    )
    ScaleA=true
    ScaleRGB=true
   )
   &ScaleSpriteSize(
    Uniform Curve Scale=&FN_FrontEndFloatModifier(
     Gameplay=&RandomRangeFloat(Minimum=1.2, Maximum=2)
     Front End=&RandomRangeFloat(Minimum=0.6, Maximum=0.9)
    )
    Uniform Curve Sprite Scale=DataInterface() {propertyValues {Curve {keys [
        {arriveTangent=3.1344, interpMode=RCIM_Cubic, leaveTangent=3.1344, tangentMode=RCTM_User, time=0, value=0}
        {tangentMode=RCTM_User, time=1, value=1}
       ]}}}
   )
   /SetVariables(
    Particles.StarColor:LinearColor=&MultiplyLinearColor(Linear Color=$Particles.Color, Scale Factor=&MakeVector4_FromLinearColor(LinearColor=<rgba 1 1 1 1>))
    Particles.StarRotation:NiagaraFloat=&RandomRangeFloat(Minimum=0, Maximum=360)
    Particles.StarSize:Vector2f=&Multiply_Vector2(A=$Particles.SpriteSize, B=<2 2>)
   ) {isSetParameters=true}
   /SetVariables(
    Particles.StreaksColor:LinearColor=&MultiplyLinearColor(
     Linear Color=$Particles.Color
     Scale Factor=&MakeVector4_FromLinearColor(LinearColor=<rgba 0.6 0.8 1 0.01>)
    )
    Particles.StreaksRotation:NiagaraFloat=&RandomRangeFloat(Minimum=0, Maximum=360)
    Particles.StreaksSize:Vector2f=&Multiply_Vector2(A=$Particles.SpriteSize, B=<7 7>)
   ) {isSetParameters=true}
   /SetVariables(
    Particles.AnamorphicColor:LinearColor=&MultiplyLinearColor(
     Linear Color=$Particles.Color
     Scale Factor=&MakeVector4_FromLinearColor(LinearColor=<rgba 0.8 0.9 1 0.01>)
    )
    Particles.AnamorphicRotation=90
    Particles.AnamorphicSize:Vector2f=&Multiply_Vector2(A=$Particles.SpriteSize, B=<3 9>)
   ) {isSetParameters=true}
   /SetVariables(
    Particles.HaloColor:LinearColor=&MultiplyLinearColor(Linear Color=$Particles.Color, Scale Factor=&MakeVector4_FromLinearColor(LinearColor=<rgba 1 1 1 0.025>))
    Particles.HaloSize:Vector2f=&Multiply_Vector2(A=$Particles.SpriteSize, B=<5 5>)
   ) {isSetParameters=true}
   /SetVariables(
    Particles.HaloColor_02:LinearColor=&MultiplyLinearColor(Linear Color=$Particles.Color, Scale Factor=&MakeVector4_FromLinearColor(LinearColor=<rgba 1 1 1 0.05>))
    Particles.HaloSize_02:Vector2f=&Multiply_Vector2(A=$Particles.SpriteSize, B=<1.5 1.5>)
   ) {isSetParameters=true}
   /SetVariables(
    Particles.AirGlowColor:LinearColor=&MultiplyLinearColor(Linear Color=$Particles.Color, Scale Factor=&MakeVector4_FromLinearColor(LinearColor=<rgba 1 1 1 0.005>))
    Particles.AirGlowSize:Vector2f=&Multiply_Vector2(A=$Particles.SpriteSize, B=<4 4>)
   ) {isSetParameters=true}
   &FN_Light_Attributes(
    Color=<rgba 3 3 3 0.2>
    Create Color Attribute=true
    Create Exponent Attribute=true
    Create Position Attribute=true
    Create Radius Attribute=true
    Enable Particle Light Enable=true
    Exponent=18
    Light Enabled=true
    Position=&AddVectorToPosition(Position=$Particles.Position, Vector=<0 -50 0>)
    Radius=140
    Scale Color=&FloatFromCurve(
     FloatCurve=DataInterface() {propertyValues {Curve {keys [
         {arriveTangent=9.0725, interpMode=RCIM_Cubic, leaveTangent=9.0725, tangentMode=RCTM_User, time=0, value=0}
         {arriveTangent=0.0246, interpMode=RCIM_Cubic, leaveTangent=0.0246, tangentMode=RCTM_User, time=0.1408, value=0.9847}
         {arriveTangent=-0.7974, interpMode=RCIM_Cubic, leaveTangent=-0.7974, tangentMode=RCTM_User, time=1, value=0}
        ]}}}
     Scale Curve=&FN_FrontEndFloatModifier(Gameplay=10, Front End=7)
    )
   )
   /SetVariables(Particles.TempPosition:Vector3f=$Particles.FN_Light_Attributes.Position) {isSetParameters=true}
   &FN_Gen5_Light(
    Attenutation:NiagaraFloat=&UniformAOrBFloat(
     A=0
     B=&RandomRangeFloat(Minimum=220, Maximum=280)
     Distribution Weight=0.66
     Evaluation Type=::NewEnumerator0
     Recalculate Random Each Loop=true
     Override Seed=false
    )
    Exponent=8
    Intensity:NiagaraFloat=&FNTimeOfDayScalar(Night Scale=5, Morning Scale=6, Day Scale=8, Evening Scale=6)
   )
   /SetVariables(
    Particles.ConeColor:LinearColor=&Multiply_LinearColorByFloat(
     Linear Color=&ColorFromCurve(ColorCurve=DataInterface() {propertyValues {
        AlphaCurve {keys [<at 0 0>,<at 0.2098 1 RCIM_Cubic>,<at 1 0 RCIM_Cubic>]}
        BlueCurve {keys [<at 0.2079 1 RCIM_Cubic>]}
        GreenCurve {keys [<at 0.2079 1 RCIM_Cubic>]}
        RedCurve {keys [<at 0.2079 1 RCIM_Cubic>]}
       }})
     Float=&FN_FrontEndFloatModifier(Gameplay=1, Front End=0.4)
    )
    Particles.ConeScale:Vector3f=&Multiply_Vector(
     A=<1 1 0.35>
     B=&VectorFromFloat(Value=&FloatFromCurve(FloatCurve=DataInterface() {propertyValues {Curve {keys [{arriveTangent=2.1576, interpMode=RCIM_Cubic, leaveTangent=2.1576, tangentMode=RCTM_User, time=0, value=0}, <at 1 1>]}}}))
    )
   ) {isSetParameters=true}
  ]
  rendererClasses [
   &NiagaraSpriteRendererProperties
   &NiagaraLightRendererProperties
   &NiagaraComponentRendererProperties
   &NiagaraMeshRendererProperties
  ]
  renderers [
   {rendererClass=&NiagaraSpriteRendererProperties}
   {rendererClass=&NiagaraSpriteRendererProperties, rendererIndex=1}
   {rendererClass=&NiagaraSpriteRendererProperties, rendererIndex=2}
   {rendererClass=&NiagaraSpriteRendererProperties, rendererIndex=3}
   {rendererClass=&NiagaraLightRendererProperties, rendererIndex=4}
   {rendererClass=&NiagaraComponentRendererProperties, rendererIndex=5}
   {rendererClass=&NiagaraMeshRendererProperties, rendererIndex=6}
   {rendererClass=&NiagaraSpriteRendererProperties, rendererIndex=7}
  ]
 }
}
summary {
 systemName=NS_Contrail_RowLiaison_02
 emitters [
  {
   emitterName=ClothandSparkles
   rendererClasses [&NiagaraMeshRendererProperties, &NiagaraSpriteRendererProperties]
  }
  {emitterName=AirFlow, rendererClasses [&NiagaraMeshRendererProperties]}
  {
   emitterName=PhotoFlashes
   rendererClasses [
    &NiagaraSpriteRendererProperties
    &NiagaraLightRendererProperties
    &NiagaraComponentRendererProperties
    &NiagaraMeshRendererProperties
   ]
  }
  {emitterName=Glitter, rendererClasses [&NiagaraSpriteRendererProperties]}
 ]
 userVariables [
  User.bIsFrontEnd {type=NiagaraBool}
  User.bIsFrontEndPreview {type=NiagaraBool}
  User.ParaGlideLeanAlpha {type=NiagaraFloat, defaultValue=1, description=""}
 ]
}
